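# MF_Erode is a material function that erodes a grayscale gradient with a noise input, as one of the MaterialGym grayscale transforms.
# It subtracts the gradient from the inverted, saturated noise, scales that difference by Contrast, then subtracts it from the gradient.
# The saturated result darkens where gradient plus noise is below one and brightens where it is above.
# Gradient falls back to the vertical UV coordinate when unconnected.
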MaterialFunction /Game/Global/Materials/MaterialGym/Materials/MaterialElements/GrayscaleTransforms/Erode/MF_Erode
graph {nodes [
  n1 = FunctionInput(Preview=ComponentMask(None=TextureCoordinate()) {g=true}) {bUsePreviewValueAsDefault=true, inputName=Gradient, inputType=FunctionInput_Scalar, previewValue=<1 0 0 1>}
  n2 = Saturate(None=FunctionInput() {bUsePreviewValueAsDefault=true, inputName=Noise, inputType=FunctionInput_Scalar, sortPriority=1})
  n3 = FunctionInput() {bUsePreviewValueAsDefault=true, inputName=Contrast, inputType=FunctionInput_Scalar, sortPriority=2, previewValue=<1 0 0 1>}
  n4 = Multiply(A=Subtract(A=OneMinus(None=@n2), B=@n1), B=@n3)
  FunctionOutput(None=Saturate(None=Subtract(A=@n1, B=@n4)))
 ]}
