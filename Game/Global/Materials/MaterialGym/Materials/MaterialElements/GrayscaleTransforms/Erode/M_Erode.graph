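# M_Erode is a Default Lit material in the MaterialGym grayscale transform elements that exercises the MF_Erode material function.
# Time scales a constant vector that offsets the UVs, scrolling an atmospheric cloud noise texture along V.
# Its red channel feeds the function's Noise input, and a scalar parameter, ErodeContrast, drives its Contrast.
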Material /Game/Global/Materials/MaterialGym/Materials/MaterialElements/GrayscaleTransforms/Erode/M_Erode
graph {
 properties {bAutomaticallySetUsageInEditor=false, usedShadingModels=Default Lit}
 nodes [
  n1 = TextureCoordinate()
  n2 = Multiply(A=Time(), B=Constant2Vector() {g=0.2}) {constB=12}
  MaterialFunctionCall!(
   Gradient
   Noise=TextureSample(UVs=Add(A=@n1, B=@n2)) {texture=&AtmosphericCloudNoise01}.R
   Contrast=ScalarParameter() {defaultValue=1, parameterName=ErodeContrast}
  ) {materialFunction=&/MF_Erode}
 ]
}
